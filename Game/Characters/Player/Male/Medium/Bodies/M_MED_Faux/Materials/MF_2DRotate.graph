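# MF_2DRotate is a library-exposed material function that rotates a UV coordinate or 2D vector by a scalar Angle.
# It splits the vector, which defaults to texture coordinates, and applies the standard sine and cosine rotation matrix before re-appending the components.
# Rotation pivots on the origin rather than the UV center.
# Sine and Cosine keep their default period, so Angle counts full turns, not radians.
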MaterialFunction /Game/Characters/Player/Male/Medium/Bodies/M_MED_Faux/Materials/MF_2DRotate
graph {
 properties {bExposeToLibrary=true}
 nodes [
  n1 = FunctionInput() {bUsePreviewValueAsDefault=true, inputName=Angle, inputType=FunctionInput_Scalar}
  n2 = Cosine(None=@n1)
  n3 = MaterialFunctionCall(Float2=FunctionInput(Preview=TextureCoordinate()) {bUsePreviewValueAsDefault=true, inputName=UV/2DVector, inputType=FunctionInput_Vector2}) {materialFunction=&BreakOutFloat2Components}
  n4 = Multiply(A=@n2, B=@n3.R)
  n5 = Sine(None=@n1)
  n6 = Multiply(A=@n5, B=@n3.G)
  n7 = Multiply(A=@n5, B=@n3.R)
  n8 = Multiply(A=@n2, B=@n3.G)
  FunctionOutput(None=AppendVector(A=Subtract(A=@n4, B=@n6), B=Add(A=@n7, B=@n8)))
 ]
}
