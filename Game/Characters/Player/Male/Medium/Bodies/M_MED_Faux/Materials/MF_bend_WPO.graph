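# Material function for the player's medium male faux body that computes a world position offset bending pre-skinned vertices along an arc.
# Height above the Min input, normalized against FA_maxHeight, scales the bend, with arc radius derived from that span and the angle.
# TexCoord U flips the bend direction at 0.5, and angle is clamped above zero.
# It outputs the local offset transformed to world space, plus height above Min.
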MaterialFunction /Game/Characters/Player/Male/Medium/Bodies/M_MED_Faux/Materials/MF_bend_WPO
graph {nodes [
  Constant() {r=90}
  n1 = NamedRerouteUsage()
  n2 = If(A=NamedRerouteUsage(), B=NamedRerouteUsage(), "A > B"=NamedRerouteUsage(), "A == B"=@n1, "A < B"=@n1)
  n3 = NamedRerouteUsage()
  FunctionOutput(None=Transform(None=Subtract(A=@n2, B=@n3)) {transformSourceType=TRANSFORMSOURCE_Local})
  FunctionOutput(None=NamedRerouteUsage()) {outputName=ZMinusmin}
  n4 = PreSkinnedPosition()
  n5 = FunctionInput(Preview=ShaderStageSwitch(PixelShader=VertexInterpolator(VS=@n4).PS, VertexShader=@n4)) {bUsePreviewValueAsDefault=true, inputName=PreskinnedPosition}
  n6 = Add(A=NamedRerouteDeclaration(None=NamedRerouteDeclaration(None=@n5) {name=PreSkinnedLocalPosition, nodeColor=<rgba 0 1 0.2969 1>}) {name=localPosition, nodeColor=<rgba 0.6875 0 1 1>}) {constB=0}
  n7 = MaterialFunctionCall(Color=@n6) {materialFunction=&SplitComponents}
  NamedRerouteDeclaration(None=@n7.R) {name=x, nodeColor=<rgba 1 0.0469 0 1>}
  n8 = NamedRerouteUsage()
  NamedRerouteDeclaration(None=Cosine(None=@n8)) {name=cos, nodeColor=<rgba 1 0.4453 0 1>}
  NamedRerouteDeclaration(None=@n6) {name=Point, nodeColor=<rgba 0.4766 0 1 1>}
  n9 = ScalarParameter() {defaultValue=165, group=Faux_WPO, parameterName=FA_maxHeight}
  n10 = NamedRerouteUsage()
  n11 = Saturate(None=Divide(A=NamedRerouteUsage(), B=Subtract(A=@n9, B=@n10)))
  NamedRerouteDeclaration(None=@n11) {name=BlendAmount, nodeColor=<rgba 0.3125 1 0 1>}
  NamedRerouteDeclaration(None=@n7.G) {name=y, nodeColor=<rgba 0 1 0.8594 1>}
  n12 = FunctionInput() {inputName=Min, inputType=FunctionInput_Scalar}
  NamedRerouteDeclaration(None=Subtract(A=@n7.B, B=NamedRerouteDeclaration(None=@n12) {name=min, nodeColor=<rgba 0.8984 0 1 1>})) {name=zMinusMin, nodeColor=<rgba 0 0.0181 1 1>}
  n13 = NamedRerouteUsage()
  NamedRerouteDeclaration(None=Multiply(A=@n11, B=Divide(A=@n13) {constB=360})) {name=a, nodeColor=<rgba 0 1 0.5312 1>}
  NamedRerouteDeclaration(None=Multiply(A=Sine(None=@n8)) {constB=-1}) {name=sin, nodeColor=<rgba 1 0 0.4922 1>}
  NamedRerouteDeclaration(None=@n7.B) {name=z, nodeColor=<rgba 0 0.1377 1 1>}
  n14 = Subtract(A=NamedRerouteUsage(), B=NamedRerouteUsage())
  n15 = Constant()
  n16 = Subtract(A=Multiply(A=@n14, B=NamedRerouteUsage()), B=Multiply(A=@n15, B=NamedRerouteUsage()))
  n17 = NamedRerouteUsage()
  n18 = Add(A=Multiply(A=@n14, B=NamedRerouteUsage()), B=Multiply(A=@n15, B=NamedRerouteUsage()))
  n19 = NamedRerouteUsage()
  NamedRerouteDeclaration(None=MaterialFunctionCall(X=Add(A=@n16, B=@n17), Y=NamedRerouteUsage(), Z=Add(A=@n18, B=@n19)) {materialFunction=&MakeFloat3}.Result) {name=bentPoint, nodeColor=<rgba 0 1 0.1562 1>}
  n20 = FunctionInput() {inputName=angle, inputType=FunctionInput_Scalar}
  n21 = ComponentMask(None=FunctionInput() {inputName=TexCoord, inputType=FunctionInput_Vector2}) {r=true}
  n22 = Constant() {r=0.5}
  n23 = Constant() {r=1}
  n24 = Constant() {r=-1}
  NamedRerouteDeclaration(None=Multiply(A=Max(A=@n20) {constB=0.01, desc=can't have an angle of 0}, B=If(A=@n21, B=@n22, "A > B"=@n23, "A == B"=@n23, "A < B"=@n24))) {name=splitAngle, nodeColor=<rgba 1 0.6562 0 1>}
  n25 = NamedRerouteUsage()
  n26 = NamedRerouteUsage()
  NamedRerouteDeclaration(None=Divide(
    A=Subtract(A=@n9, B=@n25)
    B=MaterialFunctionCall!(Degrees=@n26) {materialFunction=/Engine/Functions/Engine_MaterialFunctions01/Math/DegreesToRadians}.Radians
   )) {name=cx, nodeColor=<rgba 0 0.9766 1 1>}
 ]}
